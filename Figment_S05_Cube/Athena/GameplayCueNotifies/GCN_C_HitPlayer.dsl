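; Burst gameplay cue for the Figment Cube zapping a player, with lightning particles, a zap sound at the cube and an impact emitter at full normalized magnitude.
; It finds the cube mesh socket closest to the hit, skipping the down face.
; It aims and length-scales the first particle component toward that socket and moves the second onto it.
; It starts the player's camera shake manually, noted as a workaround for the cue not working.
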(fn OnBurst(Cascade)
  (params (MyTarget :type Actor%20Object%20Reference) (Parameters :type Gameplay%20Cue%20Parameters%20Structure%20%28by%20ref%29) (ParticleComponents :type Array%20of%20Cascade%20Particle%20System%20Component%20Object%20References) (AudioComponents :type Array%20of%20Audio%20Component%20Object%20References) (BurstCameraShakeInstance :type Camera%20Shake%20Base%20Object%20Reference) (BurstDecalInstance :type Decal%20Actor%20Object%20Reference))
  (bind (_normalizedmagnitude _rawmagnitude _effectcontext _matchedtagname _originaltag _aggregatedsourcetags _aggregatedtargettags _location _normal _instigator _effectcauser _sourceobject) (Ability|GameplayCue|BreakGameplayCueParameters Parameters))
  (bind _output (Utilities|Array|Get(acopy) ParticleComponents))
  (bind _hitlocation (LocalVariables|GetHitLocation))
  (bind _closestsocket (LocalVariables|GetClosestSocket))
  (Utilities|FlowControl|Sequence
    (:then_0
      (bind _output_get (LocalVariables|SetHitLocation _location))
      (if (== _normalizedmagnitude 1.0)
        (Effects|SpawnEmitteratLocation 0 "/Figment_S05_Cube/Athena/Prototype/Blueprints/Cube/FX/P_Lightning_Strike_Impact_Small.P_Lightning_Strike_Impact_Small" _output_get (Math|Rotator|MakeRotfromX _normal)))
      (bind _ascube_figment (Utilities|Casting|CastToCUBE_Figment _sourceobject))
      (Audio|PlaySoundatLocation _ascube_figment "/FNGameplayCues/Athena/Sounds/Events/Cube/Cube_AngryZap_Source_Cue.Cube_AngryZap_Source_Cue" (Transformation|GetWorldLocation (Variables|/Figment_S05_Cube/Athena/Prototype/Blueprints/Cube/CUBE_Figment.CUBE_Figment_C|GetAmbientAudio _ascube_figment)))
      (for _array_element (Transformation|GetAllSocketNames (Variables|/Figment_S05_Cube/Athena/Prototype/Blueprints/Cube/CUBE_Figment.CUBE_Figment_C|GetStaticMesh _ascube_figment))
        (bind (_currentstep_t_16_ea4278f841ee4e73c6861e88cb99517b _currentstep_2parttransitiontime_12_209590664625dbbf3f6200a38992da89 _currentstep_downface_19_20869c6c4df1517848398693617bdfc2) (Utilities|Struct|BreakCubeStruct (Variables|/Figment_S05_Cube/Athena/Prototype/Blueprints/Cube/CUBE_Figment.CUBE_Figment_C|GetCurrentStep_T_16_EA4278F841EE4E73C6861E88CB99517B _ascube_figment)))
        (if (Utilities|Name|NotEqual(Name) _array_element (Utilities|String|StringToName (Utilities|String|ToString(Byte) _currentstep_downface_19_20869c6c4df1517848398693617bdfc2)))
          (if (< (Math|Vector|VectorLength (- (Transformation|GetSocketLocation (Variables|/Figment_S05_Cube/Athena/Prototype/Blueprints/Cube/CUBE_Figment.CUBE_Figment_C|GetStaticMesh _ascube_figment) _array_element) _hitlocation)) (Math|Vector|VectorLength (- (Transformation|GetSocketLocation (Variables|/Figment_S05_Cube/Athena/Prototype/Blueprints/Cube/CUBE_Figment.CUBE_Figment_C|GetStaticMesh _ascube_figment) _closestsocket) _hitlocation)))
            (LocalVariables|SetClosestSocket _array_element))))
      (Transformation|SetWorldRotation _output (Math|Rotator|FindLookatRotation _hitlocation (Transformation|GetSocketLocation (Variables|/Figment_S05_Cube/Athena/Prototype/Blueprints/Cube/CUBE_Figment.CUBE_Figment_C|GetStaticMesh _ascube_figment) _closestsocket)) false false)
      (Effects|Components|ParticleSystem|SetVectorParameter _output "Length" (Math|Vector|MakeVector (/ (Math|Vector|VectorLength (- (Transformation|GetSocketLocation (Variables|/Figment_S05_Cube/Athena/Prototype/Blueprints/Cube/CUBE_Figment.CUBE_Figment_C|GetStaticMesh _ascube_figment) _closestsocket) _hitlocation)) 1e+03) 2.0 2.0))
      (Transformation|SetWorldTransform (Utilities|Array|Get(acopy) ParticleComponents 1) (Transformation|GetSocketTransform (Variables|/Figment_S05_Cube/Athena/Prototype/Blueprints/Cube/CUBE_Figment.CUBE_Figment_C|GetStaticMesh _ascube_figment) (LocalVariables|GetClosestSocket) "RTS_World") false true))
    (:then_1
      ;[1 1 1] Camera shake because GC isn't working
      (bind _asfort_player_pawn (Utilities|Casting|CastToFortPlayerPawn MyTarget)
        (:then
          (Game|Feedback|ClientStartCameraShake :self (Fortnite|GetFortPlayerControllerfromActor _asfort_player_pawn) :Shake "/FNGameplayCues/Effects/Fort_Effects/Camera_Effects/BP_CameraShake_Cube_Smack.BP_CameraShake_Cube_Smack_C" :Scale 1.0 :PlaySpace "CameraLocal"))
        (:CastFailed
          _))
      (return))))
