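; Gameplay cue notify that plays burst feedback when a player is bounced by the Cube lake.
; On burst it starts a smack camera shake on the target pawn's controller, then stores a cue location using the cue's X/Y and a hard-coded Z height.
; It moves a particle component there, plays the surface launch sound and shows a footstep-type sound indicator filtered to hostile and neutral affiliations relative to the player's team.
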(fn OnBurst(Cascade)
  (params (MyTarget :type Actor%20Object%20Reference) (Parameters :type Gameplay%20Cue%20Parameters%20Structure%20%28by%20ref%29) (ParticleComponents :type Array%20of%20Cascade%20Particle%20System%20Component%20Object%20References) (AudioComponents :type Array%20of%20Audio%20Component%20Object%20References) (BurstCameraShakeInstance :type Camera%20Shake%20Base%20Object%20Reference) (BurstDecalInstance :type Decal%20Actor%20Object%20Reference))
  (bind (_normalizedmagnitude _rawmagnitude _effectcontext _matchedtagname _originaltag _aggregatedsourcetags _aggregatedtargettags _location) (Ability|GameplayCue|BreakGameplayCueParameters Parameters))
  (bind _cuelocation (LocalVariables|GetCueLocation))
  (bind _asfort_player_pawn (Utilities|Casting|CastToFortPlayerPawn MyTarget))
  (Game|Feedback|ClientStartCameraShake :self (Fortnite|GetFortPlayerControllerfromActor _asfort_player_pawn) :Shake "/FNGameplayCues/Effects/Fort_Effects/Camera_Effects/BP_CameraShake_Cube_SmackL.BP_CameraShake_Cube_SmackL_C" :Scale 1.0 :PlaySpace "CameraLocal")
  (bind _output_get (LocalVariables|SetCueLocation (Math|Vector|MakeVector (.x _location) (.y _location) -816.0)))
  (Transformation|SetWorldLocation (Utilities|Array|Get(acopy) ParticleComponents) _output_get false false)
  (Audio|PlaySoundatLocation (Variables|/Script/Engine.Actor|GetRootComponent _asfort_player_pawn) "/Figment_S05_Cube/Athena/Sounds/Events/Cube/Cube_LaunchPlayer_Surface_Cue.Cube_LaunchPlayer_Surface_Cue" _cuelocation)
  (Fortnite|ShowSoundIndicatorLocalforTeam :Location _cuelocation :MaxAudibleDistance 0.0 :IndicatorType "FootStep" :Team (Team|GetActorTeam _asfort_player_pawn) :Affiliations (Utilities|Array|MakeArray "Hostile" "Neutral") :Tint "(R=0.000000,G=0.000000,B=0.000000,A=0.000000)" :SoundBase "/Figment_S05_Cube/Athena/Sounds/Events/Cube/Cube_LaunchPlayer_Surface_Cue.Cube_LaunchPlayer_Surface_Cue" :AttenuationMaxRangeMultiplier 1.0 :AttenuationIntensityMultiplier 1.0)
  (return))
